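# Physics asset for the Conscience pickaxe, providing bodies and joints for its handle, dynamic body and dangling tentacles.
# Capsules cover the handle, the body and three four-segment tentacle chains: upper left, upper right and lower right.
# The body joins the handle and anchors each chain, with both right-side roots set to swing and twist freely.
# Joints within each chain have narrow swing limits and almost no twist, so tentacles bend without spinning.
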PhysicsAsset /BRCosmetics/Weapons/FORT_Melee/Pickaxe_Conscience/Meshes/Conscience_Axe_Physics
bodies {
 Melee_TwoHanded_Handle [{shapeType=Capsule, length=54.1176, radius=4.7714, center=<0.8498 -0.3857 12.0993>, rotation=<rot -0.0286 -90 -0.0945>}]
 dyn_L_up_tentacle_01 [{shapeType=Capsule, length=1.5817, radius=1.1291, center=<2.4941 -0.0362 0.0432>, rotation=<rot 0.0928 -89.9412 -89.4726>}]
 dyn_L_up_tentacle_02 [{
   shapeType=Capsule
   length=1.0698
   radius=1.1773
   shapeName=dyn_L_up_tentacle_02_capsule
   center=<2.7609 0 0.0449>
   rotation=<rot 0.0001 -90.0003 -88.141>
  }]
 dyn_L_up_tentacle_03 [{
   shapeType=Capsule
   length=1.721
   radius=1.2165
   shapeName=dyn_L_up_tentacle_03_capsule
   center=<3.1018 0.1374 -0.0323>
   rotation=<rot 0 90.9049 -89.2601>
  }]
 dyn_L_up_tentacle_04 [{
   shapeType=Capsule
   length=2.0692
   radius=2.0291
   shapeName=dyn_L_up_tentacle_04_capsule
   center=<3.6819 0.1447 0.3109>
   rotation=<rot 0 88.689 -86.6551>
  }]
 dyn_R_bot_tentacle_01 [{
   shapeType=Capsule
   length=1.7817
   radius=1.1611
   shapeName=dyn_R_bot_tentacle_01_capsule
   center=<2.5765 0.0024 -0.055>
   rotation=<rot 0.1086 92.2641 -87.7553>
  }]
 dyn_R_bot_tentacle_02 [{
   shapeType=Capsule
   length=1.269
   radius=1.1661
   shapeName=dyn_R_bot_tentacle_02_capsule
   center=<2.7941 -0.0079 -0.2634>
   rotation=<rot 12.0495 -88.4698 -86.5769>
  }]
 dyn_R_bot_tentacle_03 [{
   shapeType=Capsule
   length=2.0197
   radius=1.1877
   shapeName=dyn_R_bot_tentacle_03_capsule
   center=<3.143 0.0427 -0.0448>
   rotation=<rot -0.9008 -88.3371 -91.6861>
  }]
 dyn_R_bot_tentacle_04 [{
   shapeType=Capsule
   length=1.9001
   radius=1.5884
   shapeName=dyn_R_bot_tentacle_04_capsule
   center=<3.235 0.0931 -0.0944>
   rotation=<rot -0.1137 -91.4248 -88.6429>
  }]
 dyn_R_up_tentacle_01 [{
   shapeType=Capsule
   length=1.4277
   radius=1.069
   shapeName=dyn_R_up_tentacle_01_capsule
   center=<2.6551 0.0485 -0.1986>
   rotation=<rot 0.4964 90.5375 -90.5131>
  }]
 dyn_R_up_tentacle_02 [{
   shapeType=Capsule
   length=1.7165
   radius=1.1472
   shapeName=dyn_R_up_tentacle_02_capsule
   center=<2.6652 0 0.0117>
   rotation=<rot 0 -90.0001 -90.2514>
  }]
 dyn_R_up_tentacle_03 [{
   shapeType=Capsule
   length=2.1601
   radius=1.1215
   shapeName=dyn_R_up_tentacle_03_capsule
   center=<3.1002 0.0298 0.0617>
   rotation=<rot 1.9084 -90.982 -89.3344>
  }]
 dyn_R_up_tentacle_04 [{
   shapeType=Capsule
   length=1.5895
   radius=1.8231
   shapeName=dyn_R_up_tentacle_04_capsule
   center=<2.9504 0.0197 -0.0473>
   rotation=<rot -0.7235 89.1204 -86.4071>
  }]
 dyn_body [
  {shapeType=Capsule, length=5.5077, radius=7.7154, center=<0.1093 0.0551 1.7154>, rotation=<rot 0.1441 -0.0726 -88.5496>}
  {shapeType=Capsule, length=10.2958, radius=4.6902, shapeName=dyn_body_capsule, center=<0 -20.4013 6.8334>, rotation=<rot 0 0 -34.8716>}
 ]
}
constraints [
 {bone1Name=dyn_body, bone2Name=Melee_TwoHanded_Handle, swing1LimitDegrees=45, swing2LimitDegrees=45, twistLimitDegrees=45}
 {bone1Name=dyn_L_up_tentacle_01, bone2Name=dyn_body, swing1LimitDegrees=45, swing2LimitDegrees=45, twistLimitDegrees=45}
 {
  bone1Name=dyn_R_bot_tentacle_01
  bone2Name=dyn_body
  swing1LimitDegrees=45
  swing1Motion=Free
  swing2LimitDegrees=45
  swing2Motion=Free
  twistLimitDegrees=45
  twistMotion=Free
 }
 {bone1Name=dyn_R_bot_tentacle_02, bone2Name=dyn_R_bot_tentacle_01, swing1LimitDegrees=15, swing2LimitDegrees=15, twistLimitDegrees=0.1}
 {bone1Name=dyn_R_bot_tentacle_03, bone2Name=dyn_R_bot_tentacle_02, swing1LimitDegrees=15, swing2LimitDegrees=15, twistLimitDegrees=0.1}
 {bone1Name=dyn_R_bot_tentacle_04, bone2Name=dyn_R_bot_tentacle_03, swing1LimitDegrees=15, swing2LimitDegrees=15, twistLimitDegrees=0.1}
 {bone1Name=dyn_L_up_tentacle_02, bone2Name=dyn_L_up_tentacle_01, swing1LimitDegrees=15, swing2LimitDegrees=15, twistLimitDegrees=0.1}
 {bone1Name=dyn_L_up_tentacle_03, bone2Name=dyn_L_up_tentacle_02, swing1LimitDegrees=15, swing2LimitDegrees=15, twistLimitDegrees=0.1}
 {bone1Name=dyn_L_up_tentacle_04, bone2Name=dyn_L_up_tentacle_03, swing1LimitDegrees=15, swing2LimitDegrees=15, twistLimitDegrees=0.1}
 {
  bone1Name=dyn_R_up_tentacle_01
  bone2Name=dyn_body
  swing1LimitDegrees=45
  swing1Motion=Free
  swing2LimitDegrees=45
  swing2Motion=Free
  twistLimitDegrees=45
  twistMotion=Free
 }
 {bone1Name=dyn_R_up_tentacle_02, bone2Name=dyn_R_up_tentacle_01, swing1LimitDegrees=15, swing2LimitDegrees=15, twistLimitDegrees=0.1}
 {bone1Name=dyn_R_up_tentacle_03, bone2Name=dyn_R_up_tentacle_02, swing1LimitDegrees=15, swing2LimitDegrees=15, twistLimitDegrees=0.1}
 {bone1Name=dyn_R_up_tentacle_04, bone2Name=dyn_R_up_tentacle_03, swing1LimitDegrees=15, swing2LimitDegrees=15, twistLimitDegrees=0.1}
]
