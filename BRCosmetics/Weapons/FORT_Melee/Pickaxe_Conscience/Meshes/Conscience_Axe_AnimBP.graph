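# Animation Blueprint for the Conscience pickaxe that procedurally animates its tentacles, eyes and eyelids.
# Each update builds phase-offset sine waves for tentacles, stronger and faster when sprinting, and atan-shaped sines for darting eyes and periodic blinks.
# The AnimGraph runs zero-gravity RigidBody physics, then additive bone transforms driven by those arrays, with separate sprint and non-sprint tentacle blends.
# The noise-gravity and FInterpTo nodes do not feed the final pose.
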AnimBlueprint /BRCosmetics/Weapons/FORT_Melee/Pickaxe_Conscience/Meshes/Conscience_Axe_AnimBP
target_skeleton=&/Conscience_Axe_Skeleton
graphs [
 AnimGraph {nodes [
   n1 = Get Rot_lids() -> (Rot_lids:"Array of Float (double-precision)s")
   n2 = "Get (a copy)"(Array:"Array of Float (double-precision)s"=@n1, Dimension 1=8) -> (Output:double)
   n3 = "Get (a copy)"(Array=@n1, Dimension 1=7)
   n4 = "Get (a copy)"(Array=@n1, Dimension 1=6)
   n5 = "Get (a copy)"(Array=@n1, Dimension 1=5)
   n6 = "Get (a copy)"(Array=@n1, Dimension 1=3)
   n7 = "Get (a copy)"(Array=@n1, Dimension 1=2)
   n8 = "Get (a copy)"(Array=@n1, Dimension 1=1)
   n9 = "Get (a copy)"(Array=@n1)
   n10 = Get Rot_eyes() -> (Rot_eyes:Array of Rotators)
   n11 = Get Rot_tentacles() -> (Rot_tentacles:Array of Rotators)
   AnimGraphNode_SubInput_0 = Input Pose()
   n12 = RigidBody!(ComponentPose=Local To Component(LocalPose=@AnimGraphNode_SubInput_0), Alpha, OverrideWorldGravity:Vector="0,0,0", ExternalForce) -> (Pose) {
    bOverrideWorldGravity=true
    overridePhysicsAsset=&/Conscience_Axe_Physics
    simulationSpace=BaseBoneSpace
    alphaBoolBlend {blendOption=Linear}
    baseBoneRef {boneName=b_AR_Root}
    componentLinearAccScale=<0.1 0.1 0.1>
    componentLinearVelScale=<0.1 0.1 0.1>
    simSpaceSettings {worldAlpha=0.1}
   }
   n13 = "Transform (Modify) Bone - Bone: dyn_L_up_tentacle_02"!(
    ComponentPose=@n12
    bAlphaBoolEnabled:bool=NOT Boolean(A=Get bIsSprinting())
    Translation="0,0,0"
    Rotation="Get (a copy)"(Array:Array of Rotators=@n11) -> (Output:Rotator)
    Scale="1,1,1"
   ) -> (Pose) {
    alphaInputType=Bool
    rotationMode=BMM_Additive
    rotationSpace=BCS_BoneSpace
    alphaBoolBlend {blendInTime=0.5, blendOption=Linear, blendOutTime=0.5}
    boneToModify {boneName=dyn_L_up_tentacle_02}
    rotation {yaw=7.072}
   }
   n14 = "Transform (Modify) Bone - Bone: dyn_L_up_tentacle_03"!(
    ComponentPose=@n13
    bAlphaBoolEnabled:bool=NOT Boolean(A=Get bIsSprinting())
    Translation="0,0,0"
    Rotation="Get (a copy)"(Array=@n11, Dimension 1=1)
    Scale="1,1,1"
   ) -> (Pose) {
    alphaInputType=Bool
    rotationMode=BMM_Additive
    rotationSpace=BCS_BoneSpace
    alphaBoolBlend {blendInTime=0.5, blendOption=Linear, blendOutTime=0.5}
    boneToModify {boneName=dyn_L_up_tentacle_03}
    rotation {yaw=7.072}
   }
   n15 = "Transform (Modify) Bone - Bone: dyn_L_up_tentacle_04"!(
    ComponentPose=@n14
    bAlphaBoolEnabled:bool=NOT Boolean(A=Get bIsSprinting())
    Translation="0,0,0"
    Rotation="Get (a copy)"(Array=@n11, Dimension 1=2)
    Scale="1,1,1"
   ) -> (Pose) {
    alphaInputType=Bool
    rotationMode=BMM_Additive
    rotationSpace=BCS_BoneSpace
    alphaBoolBlend {blendInTime=0.5, blendOption=Linear, blendOutTime=0.5}
    boneToModify {boneName=dyn_L_up_tentacle_04}
    rotation {yaw=7.072}
   }
   n16 = "Transform (Modify) Bone - Bone: dyn_R_up_tentacle_02"!(
    ComponentPose=@n15
    bAlphaBoolEnabled:bool=NOT Boolean(A=Get bIsSprinting())
    Translation="0,0,0"
    Rotation="Get (a copy)"(Array=@n11, Dimension 1=3)
    Scale="1,1,1"
   ) -> (Pose) {
    alphaInputType=Bool
    rotationMode=BMM_Additive
    rotationSpace=BCS_BoneSpace
    alphaBoolBlend {blendInTime=0.5, blendOption=Linear, blendOutTime=0.5}
    boneToModify {boneName=dyn_R_up_tentacle_02}
    rotation {yaw=7.072}
   }
   n17 = "Transform (Modify) Bone - Bone: dyn_R_up_tentacle_03"!(
    ComponentPose=@n16
    bAlphaBoolEnabled:bool=NOT Boolean(A=Get bIsSprinting())
    Translation="0,0,0"
    Rotation="Get (a copy)"(Array=@n11, Dimension 1=4)
    Scale="1,1,1"
   ) -> (Pose) {
    alphaInputType=Bool
    rotationMode=BMM_Additive
    rotationSpace=BCS_BoneSpace
    alphaBoolBlend {blendInTime=0.5, blendOption=Linear, blendOutTime=0.5}
    boneToModify {boneName=dyn_R_up_tentacle_03}
    rotation {yaw=7.072}
   }
   n18 = "Transform (Modify) Bone - Bone: dyn_R_up_tentacle_04"!(
    ComponentPose=@n17
    bAlphaBoolEnabled:bool=NOT Boolean(A=Get bIsSprinting())
    Translation="0,0,0"
    Rotation="Get (a copy)"(Array=@n11, Dimension 1=5)
    Scale="1,1,1"
   ) -> (Pose) {
    alphaInputType=Bool
    rotationMode=BMM_Additive
    rotationSpace=BCS_BoneSpace
    alphaBoolBlend {blendInTime=0.5, blendOption=Linear, blendOutTime=0.5}
    boneToModify {boneName=dyn_R_up_tentacle_04}
    rotation {yaw=7.072}
   }
   n19 = "Transform (Modify) Bone - Bone: dyn_R_bot_tentacle_02"!(
    ComponentPose=@n18
    bAlphaBoolEnabled:bool=NOT Boolean(A=Get bIsSprinting())
    Translation="0,0,0"
    Rotation="Get (a copy)"(Array=@n11, Dimension 1=6)
    Scale="1,1,1"
   ) -> (Pose) {
    alphaInputType=Bool
    rotationMode=BMM_Additive
    rotationSpace=BCS_BoneSpace
    alphaBoolBlend {blendInTime=0.5, blendOption=Linear, blendOutTime=0.5}
    boneToModify {boneName=dyn_R_bot_tentacle_02}
    rotation {yaw=7.072}
   }
   n20 = "Transform (Modify) Bone - Bone: dyn_R_bot_tentacle_03"!(
    ComponentPose=@n19
    bAlphaBoolEnabled:bool=NOT Boolean(A=Get bIsSprinting())
    Translation="0,0,0"
    Rotation="Get (a copy)"(Array=@n11, Dimension 1=7)
    Scale="1,1,1"
   ) -> (Pose) {
    alphaInputType=Bool
    rotationMode=BMM_Additive
    rotationSpace=BCS_BoneSpace
    alphaBoolBlend {blendInTime=0.5, blendOption=Linear, blendOutTime=0.5}
    boneToModify {boneName=dyn_R_bot_tentacle_03}
    rotation {yaw=7.072}
   }
   n21 = "Transform (Modify) Bone - Bone: dyn_R_bot_tentacle_04"!(
    ComponentPose=@n20
    bAlphaBoolEnabled:bool=NOT Boolean(A=Get bIsSprinting())
    Translation="0,0,0"
    Rotation="Get (a copy)"(Array=@n11, Dimension 1=8)
    Scale="1,1,1"
   ) -> (Pose) {
    alphaInputType=Bool
    rotationMode=BMM_Additive
    rotationSpace=BCS_BoneSpace
    alphaBoolBlend {blendInTime=0.5, blendOption=Linear, blendOutTime=0.5}
    boneToModify {boneName=dyn_R_bot_tentacle_04}
    rotation {yaw=7.072}
   }
   n22 = "Transform (Modify) Bone - Bone: dyn_L_up_tentacle_02"(ComponentPose=@n21, bAlphaBoolEnabled=Get bIsSprinting(), Translation="0,0,0", Rotation="Get (a copy)"(Array=@n11), Scale="1,1,1") {
    alphaInputType=Bool
    rotationMode=BMM_Additive
    rotationSpace=BCS_BoneSpace
    alphaBoolBlend {blendInTime=3, blendOption=Linear, blendOutTime=1}
    boneToModify {boneName=dyn_L_up_tentacle_02}
    rotation {yaw=7.072}
   }
   n23 = "Transform (Modify) Bone - Bone: dyn_L_up_tentacle_03"(
    ComponentPose=@n22
    bAlphaBoolEnabled=Get bIsSprinting()
    Translation="0,0,0"
    Rotation="Get (a copy)"(Array=@n11, Dimension 1=1)
    Scale="1,1,1"
   ) {
    alphaInputType=Bool
    rotationMode=BMM_Additive
    rotationSpace=BCS_BoneSpace
    alphaBoolBlend {blendInTime=3, blendOption=Linear, blendOutTime=1}
    boneToModify {boneName=dyn_L_up_tentacle_03}
    rotation {yaw=7.072}
   }
   n24 = "Transform (Modify) Bone - Bone: dyn_L_up_tentacle_04"(
    ComponentPose=@n23
    bAlphaBoolEnabled=Get bIsSprinting()
    Translation="0,0,0"
    Rotation="Get (a copy)"(Array=@n11, Dimension 1=2)
    Scale="1,1,1"
   ) {
    alphaInputType=Bool
    rotationMode=BMM_Additive
    rotationSpace=BCS_BoneSpace
    alphaBoolBlend {blendInTime=3, blendOption=Linear, blendOutTime=1}
    boneToModify {boneName=dyn_L_up_tentacle_04}
    rotation {yaw=7.072}
   }
   n25 = "Transform (Modify) Bone - Bone: dyn_R_up_tentacle_02"(
    ComponentPose=@n24
    bAlphaBoolEnabled=Get bIsSprinting()
    Translation="0,0,0"
    Rotation="Get (a copy)"(Array=@n11, Dimension 1=3)
    Scale="1,1,1"
   ) {
    alphaInputType=Bool
    rotationMode=BMM_Additive
    rotationSpace=BCS_BoneSpace
    alphaBoolBlend {blendInTime=3, blendOption=Linear, blendOutTime=1}
    boneToModify {boneName=dyn_R_up_tentacle_02}
    rotation {yaw=7.072}
   }
   n26 = "Transform (Modify) Bone - Bone: dyn_R_up_tentacle_03"(
    ComponentPose=@n25
    bAlphaBoolEnabled=Get bIsSprinting()
    Translation="0,0,0"
    Rotation="Get (a copy)"(Array=@n11, Dimension 1=4)
    Scale="1,1,1"
   ) {
    alphaInputType=Bool
    rotationMode=BMM_Additive
    rotationSpace=BCS_BoneSpace
    alphaBoolBlend {blendInTime=3, blendOption=Linear, blendOutTime=1}
    boneToModify {boneName=dyn_R_up_tentacle_03}
    rotation {yaw=7.072}
   }
   n27 = "Transform (Modify) Bone - Bone: dyn_R_up_tentacle_04"(
    ComponentPose=@n26
    bAlphaBoolEnabled=Get bIsSprinting()
    Translation="0,0,0"
    Rotation="Get (a copy)"(Array=@n11, Dimension 1=5)
    Scale="1,1,1"
   ) {
    alphaInputType=Bool
    rotationMode=BMM_Additive
    rotationSpace=BCS_BoneSpace
    alphaBoolBlend {blendInTime=3, blendOption=Linear, blendOutTime=1}
    boneToModify {boneName=dyn_R_up_tentacle_04}
    rotation {yaw=7.072}
   }
   n28 = "Transform (Modify) Bone - Bone: dyn_R_bot_tentacle_02"(
    ComponentPose=@n27
    bAlphaBoolEnabled=Get bIsSprinting()
    Translation="0,0,0"
    Rotation="Get (a copy)"(Array=@n11, Dimension 1=6)
    Scale="1,1,1"
   ) {
    alphaInputType=Bool
    rotationMode=BMM_Additive
    rotationSpace=BCS_BoneSpace
    alphaBoolBlend {blendInTime=3, blendOption=Linear, blendOutTime=1}
    boneToModify {boneName=dyn_R_bot_tentacle_02}
    rotation {yaw=7.072}
   }
   n29 = "Transform (Modify) Bone - Bone: dyn_R_bot_tentacle_03"(
    ComponentPose=@n28
    bAlphaBoolEnabled=Get bIsSprinting()
    Translation="0,0,0"
    Rotation="Get (a copy)"(Array=@n11, Dimension 1=7)
    Scale="1,1,1"
   ) {
    alphaInputType=Bool
    rotationMode=BMM_Additive
    rotationSpace=BCS_BoneSpace
    alphaBoolBlend {blendInTime=3, blendOption=Linear, blendOutTime=1}
    boneToModify {boneName=dyn_R_bot_tentacle_03}
    rotation {yaw=7.072}
   }
   n30 = "Transform (Modify) Bone - Bone: dyn_R_bot_tentacle_04"(
    ComponentPose=@n29
    bAlphaBoolEnabled=Get bIsSprinting()
    Translation="0,0,0"
    Rotation="Get (a copy)"(Array=@n11, Dimension 1=8)
    Scale="1,1,1"
   ) {
    alphaInputType=Bool
    rotationMode=BMM_Additive
    rotationSpace=BCS_BoneSpace
    alphaBoolBlend {blendInTime=3, blendOption=Linear, blendOutTime=1}
    boneToModify {boneName=dyn_R_bot_tentacle_04}
    rotation {yaw=7.072}
   }
   n31 = "Transform (Modify) Bone - Bone: dyn_central_eye"!(ComponentPose=@n30, bAlphaBoolEnabled:bool=True, Translation="0,0,0", Rotation="Get (a copy)"(Array=@n10), Scale="1,1,1") -> (Pose) {
    alphaInputType=Bool
    rotationMode=BMM_Additive
    rotationSpace=BCS_BoneSpace
    alphaBoolBlend {blendOption=Linear}
    boneToModify {boneName=dyn_central_eye}
    rotation {yaw=7.072}
   }
   n32 = "Transform (Modify) Bone - Bone: dyn_side_eye_01"!(ComponentPose=@n31, bAlphaBoolEnabled:bool=True, Translation="0,0,0", Rotation="Get (a copy)"(Array=@n10, Dimension 1=1), Scale="1,1,1") -> (Pose) {
    alphaInputType=Bool
    rotationMode=BMM_Additive
    rotationSpace=BCS_BoneSpace
    alphaBoolBlend {blendOption=Linear}
    boneToModify {boneName=dyn_side_eye_01}
    rotation {yaw=7.072}
   }
   n33 = "Transform (Modify) Bone - Bone: dyn_side_eye_02"!(ComponentPose=@n32, bAlphaBoolEnabled:bool=True, Translation="0,0,0", Rotation="Get (a copy)"(Array=@n10, Dimension 1=2), Scale="1,1,1") -> (Pose) {
    alphaInputType=Bool
    rotationMode=BMM_Additive
    rotationSpace=BCS_BoneSpace
    alphaBoolBlend {blendOption=Linear}
    boneToModify {boneName=dyn_side_eye_02}
    rotation {yaw=7.072}
   }
   n34 = "Transform (Modify) Bone - Bone: dyn_side_eye_03"!(ComponentPose=@n33, bAlphaBoolEnabled:bool=True, Translation="0,0,0", Rotation="Get (a copy)"(Array=@n10, Dimension 1=3), Scale="1,1,1") -> (Pose) {
    alphaInputType=Bool
    rotationMode=BMM_Additive
    rotationSpace=BCS_BoneSpace
    alphaBoolBlend {blendOption=Linear}
    boneToModify {boneName=dyn_side_eye_03}
    rotation {yaw=7.072}
   }
   n35 = "Transform (Modify) Bone - Bone: dyn_side_eye_04"!(ComponentPose=@n34, bAlphaBoolEnabled:bool=True, Translation="0,0,0", Rotation="Get (a copy)"(Array=@n10, Dimension 1=4), Scale="1,1,1") -> (Pose) {
    alphaInputType=Bool
    rotationMode=BMM_Additive
    rotationSpace=BCS_BoneSpace
    alphaBoolBlend {blendOption=Linear}
    boneToModify {boneName=dyn_side_eye_04}
    rotation {yaw=7.072}
   }
   n36 = "Transform (Modify) Bone - Bone: dyn_side_eye_05"!(ComponentPose=@n35, bAlphaBoolEnabled:bool=True, Translation="0,0,0", Rotation="Get (a copy)"(Array=@n10, Dimension 1=5), Scale="1,1,1") -> (Pose) {
    alphaInputType=Bool
    rotationMode=BMM_Additive
    rotationSpace=BCS_BoneSpace
    alphaBoolBlend {blendOption=Linear}
    boneToModify {boneName=dyn_side_eye_05}
    rotation {yaw=7.072}
   }
   n37 = "Transform (Modify) Bone - Bone: dyn_L_up_eye"!(ComponentPose=@n36, bAlphaBoolEnabled:bool=True, Translation="0,0,0", Rotation="Get (a copy)"(Array=@n10, Dimension 1=6), Scale="1,1,1") -> (Pose) {
    alphaInputType=Bool
    rotationMode=BMM_Additive
    rotationSpace=BCS_BoneSpace
    alphaBoolBlend {blendOption=Linear}
    boneToModify {boneName=dyn_L_up_eye}
    rotation {yaw=7.072}
   }
   n38 = "Transform (Modify) Bone - Bone: dyn_R_up_eye"!(ComponentPose=@n37, bAlphaBoolEnabled:bool=True, Translation="0,0,0", Rotation="Get (a copy)"(Array=@n10, Dimension 1=7), Scale="1,1,1") -> (Pose) {
    alphaInputType=Bool
    rotationMode=BMM_Additive
    rotationSpace=BCS_BoneSpace
    alphaBoolBlend {blendOption=Linear}
    boneToModify {boneName=dyn_R_up_eye}
    rotation {yaw=7.072}
   }
   n39 = "Transform (Modify) Bone - Bone: dyn_R_bot_eye"!(ComponentPose=@n38, bAlphaBoolEnabled:bool=True, Translation="0,0,0", Rotation="Get (a copy)"(Array=@n10, Dimension 1=8), Scale="1,1,1") -> (Pose) {
    alphaInputType=Bool
    rotationMode=BMM_Additive
    rotationSpace=BCS_BoneSpace
    alphaBoolBlend {blendOption=Linear}
    boneToModify {boneName=dyn_R_bot_eye}
    rotation {yaw=7.072}
   }
   n40 = "Transform (Modify) Bone - Bone: dyn_side_eye_01_blink_up"(ComponentPose=@n39, Alpha=@n9, Translation="0,0,0", Rotation="45,0.000,0.000", Scale="1,1,1") {
    rotationMode=BMM_Additive
    rotationSpace=BCS_BoneSpace
    alphaBoolBlend {blendOption=Linear}
    boneToModify {boneName=dyn_side_eye_01_blink_up}
    rotation {pitch=45}
   }
   n41 = "Transform (Modify) Bone - Bone: dyn_side_eye_01_blink_bot"(ComponentPose=@n40, Alpha=@n9, Translation="0,0,0", Rotation="-45,0.000,-0.000", Scale="1,1,1") {
    rotationMode=BMM_Additive
    rotationSpace=BCS_BoneSpace
    alphaBoolBlend {blendOption=Linear}
    boneToModify {boneName=dyn_side_eye_01_blink_bot}
    rotation {pitch=-45}
   }
   n42 = "Transform (Modify) Bone - Bone: dyn_side_eye_01_inside"(ComponentPose=@n41, Alpha=@n9, Translation="-0.2,0,0", Rotation="0,0,0", Scale="1,1,1") {
    rotationMode=BMM_Additive
    rotationSpace=BCS_BoneSpace
    translationMode=BMM_Additive
    translationSpace=BCS_BoneSpace
    alphaBoolBlend {blendOption=Linear}
    boneToModify {boneName=dyn_side_eye_01_inside}
    translation {x=-0.2}
   }
   n43 = "Transform (Modify) Bone - Bone: dyn_side_eye_02_blink_up"(ComponentPose=@n42, Alpha=@n8, Translation="0,0,0", Rotation="34.662,0.000,0.000", Scale="1,1,1") {
    rotationMode=BMM_Additive
    rotationSpace=BCS_BoneSpace
    alphaBoolBlend {blendOption=Linear}
    boneToModify {boneName=dyn_side_eye_02_blink_up}
    rotation=<rot 34.662 0 0.0002>
   }
   n44 = "Transform (Modify) Bone - Bone: dyn_side_eye_02_blink_bot"(ComponentPose=@n43, Alpha=@n8, Translation="0,0,0", Rotation="-31.176,-0.000,0.000", Scale="1,1,1") {
    rotationMode=BMM_Additive
    rotationSpace=BCS_BoneSpace
    alphaBoolBlend {blendOption=Linear}
    boneToModify {boneName=dyn_side_eye_02_blink_bot}
    rotation {pitch=-31.176}
   }
   n45 = "Transform (Modify) Bone - Bone: dyn_side_eye_02_inside"(ComponentPose=@n44, Alpha=@n8, Translation="-0.6,0,0", Rotation="0,0,0", Scale="1,1,1") {
    rotationMode=BMM_Additive
    rotationSpace=BCS_BoneSpace
    translationMode=BMM_Additive
    translationSpace=BCS_BoneSpace
    alphaBoolBlend {blendOption=Linear}
    boneToModify {boneName=dyn_side_eye_02_inside}
    translation {x=-0.6}
   }
   n46 = "Transform (Modify) Bone - Bone: dyn_side_eye_03_blink_up"(ComponentPose=@n45, Alpha=@n7, Translation="0,0,0", Rotation="48,0.000,0.000", Scale="1,1,1") {
    rotationMode=BMM_Additive
    rotationSpace=BCS_BoneSpace
    alphaBoolBlend {blendOption=Linear}
    boneToModify {boneName=dyn_side_eye_03_blink_up}
    rotation {pitch=48}
   }
   n47 = "Transform (Modify) Bone - Bone: dyn_side_eye_03_blink_bot"(ComponentPose=@n46, Alpha=@n7, Translation="0,0,0", Rotation="-48,0.000,-0.000", Scale="1,1,1") {
    rotationMode=BMM_Additive
    rotationSpace=BCS_BoneSpace
    alphaBoolBlend {blendOption=Linear}
    boneToModify {boneName=dyn_side_eye_03_blink_bot}
    rotation {pitch=-48}
   }
   n48 = "Transform (Modify) Bone - Bone: dyn_side_eye_03_inside"(ComponentPose=@n47, Alpha=@n7, Translation="-0.2,0,0", Rotation="0,0,0", Scale="1,1,1") {
    rotationSpace=BCS_BoneSpace
    translationMode=BMM_Additive
    translationSpace=BCS_BoneSpace
    alphaBoolBlend {blendOption=Linear}
    boneToModify {boneName=dyn_side_eye_03_inside}
    translation {x=-0.2}
   }
   n49 = "Transform (Modify) Bone - Bone: dyn_side_eye_04_blink_up"(ComponentPose=@n48, Alpha=@n6, Translation="0,0,0", Rotation="-37.353,0.000,-0.000", Scale="1,1,1") {
    rotationMode=BMM_Additive
    rotationSpace=BCS_BoneSpace
    alphaBoolBlend {blendOption=Linear}
    boneToModify {boneName=dyn_side_eye_04_blink_up}
    rotation {pitch=-37.353}
   }
   n50 = "Transform (Modify) Bone - Bone: dyn_side_eye_04_blink_bot"(ComponentPose=@n49, Alpha=@n6, Translation="0,0,0", Rotation="45.363,-0.016,-0.014", Scale="1,1,1") {
    rotationMode=BMM_Additive
    rotationSpace=BCS_BoneSpace
    alphaBoolBlend {blendOption=Linear}
    boneToModify {boneName=dyn_side_eye_04_blink_bot}
    rotation=<rot 45.363 -0.0161 -0.0141>
   }
   n51 = "Transform (Modify) Bone - Bone: dyn_side_eye_04_inside"(ComponentPose=@n50, Alpha=@n6, Translation="-0.6,0,0", Rotation="0,0,0", Scale="1,1,1") {
    rotationMode=BMM_Additive
    rotationSpace=BCS_BoneSpace
    translationMode=BMM_Additive
    translationSpace=BCS_BoneSpace
    alphaBoolBlend {blendOption=Linear}
    boneToModify {boneName=dyn_side_eye_04_inside}
    translation {x=-0.6}
   }
   n52 = "Get (a copy)"(Array:"Array of Float (double-precision)s"=@n1, Dimension 1=4) -> (Output:double)
   n53 = "Transform (Modify) Bone - Bone: dyn_side_eye_05_blink_up"(ComponentPose=@n51, Alpha=@n52, Translation="0,0,0", Rotation="-37.353,0.000,0.000", Scale="1,1,1") {
    rotationMode=BMM_Additive
    rotationSpace=BCS_BoneSpace
    alphaBoolBlend {blendOption=Linear}
    boneToModify {boneName=dyn_side_eye_05_blink_up}
    rotation {pitch=-37.353}
   }
   n54 = "Transform (Modify) Bone - Bone: dyn_side_eye_05_blink_bot"(ComponentPose=@n53, Alpha=@n52, Translation="0,0,0", Rotation="31.176,-0.000,0.000", Scale="1,1,1") {
    rotationMode=BMM_Additive
    rotationSpace=BCS_BoneSpace
    alphaBoolBlend {blendOption=Linear}
    boneToModify {boneName=dyn_side_eye_05_blink_bot}
    rotation {pitch=31.176}
   }
   n55 = "Transform (Modify) Bone - Bone: dyn_side_eye_05_inside"(ComponentPose=@n54, Alpha=@n52, Translation="-0.2,0,0", Rotation="0,0,0", Scale="1,1,1") {
    rotationMode=BMM_Additive
    rotationSpace=BCS_BoneSpace
    translationMode=BMM_Additive
    translationSpace=BCS_BoneSpace
    alphaBoolBlend {blendOption=Linear}
    boneToModify {boneName=dyn_side_eye_05_inside}
    translation {x=-0.2}
   }
   n56 = "Transform (Modify) Bone - Bone: dyn_central_blink_up"(ComponentPose=@n55, Alpha=@n5, Translation="0,0,0", Rotation="-0.000,-64.404,-0.000", Scale="1,1,1") {
    rotationMode=BMM_Additive
    rotationSpace=BCS_BoneSpace
    translationSpace=BCS_BoneSpace
    alphaBoolBlend {blendOption=Linear}
    boneToModify {boneName=dyn_central_blink_up}
    rotation {yaw=-64.404}
   }
   n57 = "Transform (Modify) Bone - Bone: dyn_central_blink_bot"(ComponentPose=@n56, Alpha=@n5, Translation="0,0,0", Rotation="0.000,64.444,-0.000", Scale="1,1,1") {
    rotationMode=BMM_Additive
    rotationSpace=BCS_BoneSpace
    alphaBoolBlend {blendOption=Linear}
    boneToModify {boneName=dyn_central_blink_bot}
    rotation=<rot 0 64.444 0>
   }
   n58 = "Transform (Modify) Bone - Bone: dyn_central_eye_inside"(ComponentPose=@n57, Alpha=@n5, Translation="-0.6,0,0", Rotation="0,0,0", Scale="1,1,1") {
    rotationMode=BMM_Additive
    rotationSpace=BCS_BoneSpace
    translationMode=BMM_Additive
    translationSpace=BCS_BoneSpace
    alphaBoolBlend {blendOption=Linear}
    boneToModify {boneName=dyn_central_eye_inside}
    translation {x=-0.6}
   }
   n59 = "Transform (Modify) Bone - Bone: dyn_L_up_eye_blink_up"(ComponentPose=@n58, Alpha=@n4, Translation="0,0,0", Rotation="-0.000,-60,-0.000", Scale="1,1,1") {
    rotationMode=BMM_Additive
    rotationSpace=BCS_BoneSpace
    alphaBoolBlend {blendOption=Linear}
    boneToModify {boneName=dyn_L_up_eye_blink_up}
    rotation {yaw=-60}
   }
   n60 = "Transform (Modify) Bone - Bone: dyn_L_up_eye_blink_bot"(ComponentPose=@n59, Alpha=@n4, Translation="0,0,0", Rotation="0,60,0.000", Scale="1,1,1") {
    rotationMode=BMM_Additive
    rotationSpace=BCS_BoneSpace
    alphaBoolBlend {blendOption=Linear}
    boneToModify {boneName=dyn_L_up_eye_blink_bot}
    rotation {yaw=60}
   }
   n61 = "Transform (Modify) Bone - Bone: dyn_L_up_eye_inside"(ComponentPose=@n60, Alpha=@n4, Translation="-1,0,0", Rotation="0,0,0", Scale="1,1,1") {
    rotationMode=BMM_Additive
    rotationSpace=BCS_BoneSpace
    translationMode=BMM_Additive
    translationSpace=BCS_BoneSpace
    alphaBoolBlend {blendOption=Linear}
    boneToModify {boneName=dyn_L_up_eye_inside}
    translation {x=-1}
   }
   n62 = "Transform (Modify) Bone - Bone: dyn_R_up_eye_blink_up"(ComponentPose=@n61, Alpha=@n3, Translation="0,0,0", Rotation="0,-60,0.000", Scale="1,1,1") {
    rotationMode=BMM_Additive
    rotationSpace=BCS_BoneSpace
    alphaBoolBlend {blendOption=Linear}
    boneToModify {boneName=dyn_R_up_eye_blink_up}
    rotation {yaw=-60}
   }
   n63 = "Transform (Modify) Bone - Bone: dyn_R_up_eye_blink_bot"(ComponentPose=@n62, Alpha=@n3, Translation="0,0,0", Rotation="0,60,0.000", Scale="1,1,1") {
    rotationMode=BMM_Additive
    rotationSpace=BCS_BoneSpace
    alphaBoolBlend {blendOption=Linear}
    boneToModify {boneName=dyn_R_up_eye_blink_bot}
    rotation {yaw=60}
   }
   n64 = "Transform (Modify) Bone - Bone: dyn_R_up_eye_inside"(ComponentPose=@n63, Alpha=@n3, Translation="-1,0,0", Rotation="0,0,0", Scale="1,1,1") {
    rotationMode=BMM_Additive
    rotationSpace=BCS_BoneSpace
    translationMode=BMM_Additive
    translationSpace=BCS_BoneSpace
    alphaBoolBlend {blendOption=Linear}
    boneToModify {boneName=dyn_R_up_eye_inside}
    translation {x=-1}
   }
   n65 = "Transform (Modify) Bone - Bone: dyn_R_bot_eye_blink_up"(ComponentPose=@n64, Alpha=@n2, Translation="0,0,0", Rotation="0,-62,0.000", Scale="1,1,1") {
    rotationMode=BMM_Additive
    rotationSpace=BCS_BoneSpace
    alphaBoolBlend {blendOption=Linear}
    boneToModify {boneName=dyn_R_bot_eye_blink_up}
    rotation {yaw=-62}
   }
   n66 = "Transform (Modify) Bone - Bone: dyn_R_bot_eye_blink_bot"(ComponentPose=@n65, Alpha=@n2, Translation="0,0,0", Rotation="0,63,0.000", Scale="1,1,1") {
    rotationMode=BMM_Additive
    rotationSpace=BCS_BoneSpace
    alphaBoolBlend {blendOption=Linear}
    boneToModify {boneName=dyn_R_bot_eye_blink_bot}
    rotation {yaw=63}
   }
   n67 = "Transform (Modify) Bone - Bone: dyn_R_bot_eye_inside"(ComponentPose=@n66, Alpha=@n2, Translation="-1.5,0,0", Rotation="0,0,0", Scale="1,1,1") {
    rotationMode=BMM_Additive
    rotationSpace=BCS_BoneSpace
    translationMode=BMM_Additive
    translationSpace=BCS_BoneSpace
    alphaBoolBlend {blendOption=Linear}
    boneToModify {boneName=dyn_R_bot_eye_inside}
    translation {x=-1.5}
   }
   Output Pose(Result=Component To Local(ComponentPose=@n67)) {layerGroup=None}
   Get Gravity overraide() -> (Gravity overraide:Vector)
  ]}
 EventGraph {nodes [
   n1 = Get Freq_noise() -> (Freq_noise:double)
   n2 = Get Amp_noise() -> (Amp_noise:double)
   n3 = Random Vector Function!(
    execute
    CharacterPartAnimInstance:Custom Character Part Anim Instance Object Reference
    SkelMeshComponent:Mesh Component Object Reference=GetOwnerSkeletalMeshComponent()
    JointName:Name=None
    GameTimeInSeconds:double=GetGameTimeInSeconds()
    TimeOffset:double=0
    PerlinScale:double=1
    Frequency:Vector=MakeVector(X=@n1, Y=@n1, Z=@n1)
    Amplitude:Vector=MakeVector(X=@n2, Y=@n2, Z=@n2)
    Offset:Vector="0,0,0"
   ) -> (then, VectorOut:Vector)
   > Set Gravity overraide(Gravity overraide:Vector=@n3) -> (Output_Get:Vector)
   n4 = Event BlueprintUpdateAnimation()
   > n5 = Branch(Condition=Get bIsSprinting())
   > RetriggerableDelay(Duration=0.05)
   > Set Amp_sin(Amp_sin:double=20) -> (Output_Get:double)
   > n6 = Set Freq_sin(Freq_sin:double=4) -> (Output_Get:double)
   n7 = GetRealTimeSeconds()
   n8 = float * float(A=@n7, B=Get Freq_sin() -> (Freq_sin:double))
   n9 = float + float(A=@n8, B=1000)
   n10 = float + float(A=@n9, B=1500)
   n11 = float + float(A=float * float(A=@n10, B=1), B=1000)
   n12 = float + float(A=float * float(A=@n10, B=1), B=1000)
   n13 = MakeRotator(Pitch=float * float(A="Sin (Radians)"(A=@n11), B=3), Yaw=float * float(A="Sin (Radians)"(A=@n12), B=3))
   n14 = Get Amp_sin() -> (Amp_sin:double)
   n15 = float + float(A=float * float(A=@n10, B=1), B=50)
   n16 = float + float(A=float * float(A=@n10, B=1), B=50)
   n17 = MakeRotator(Pitch=float * float(A="Sin (Radians)"(A=@n15), B=1), Yaw=float * float(A="Sin (Radians)"(A=@n16), B=1))
   n18 = float + float(A=float * float(A=@n10, B=1), B=500)
   n19 = float + float(A=float * float(A=@n10, B=1), B=500)
   n20 = MakeRotator(Pitch=float * float(A="Sin (Radians)"(A=@n18), B=2), Yaw=float * float(A="Sin (Radians)"(A=@n19), B=2))
   n21 = float + float(A=float * float(A=@n8, B=1), B=50)
   n22 = float + float(A=float * float(A=@n8, B=1), B=50)
   n23 = MakeRotator(Pitch=float * float(A="Sin (Radians)"(A=@n21), B=1), Yaw=float * float(A="Sin (Radians)"(A=@n22), B=1))
   n24 = float + float(A=float * float(A=@n8, B=1), B=500)
   n25 = float + float(A=float * float(A=@n8, B=1), B=500)
   n26 = MakeRotator(Pitch=float * float(A="Sin (Radians)"(A=@n24), B=2), Yaw=float * float(A="Sin (Radians)"(A=@n25), B=2))
   n27 = float + float(A=float * float(A=@n8, B=1), B=1000)
   n28 = float + float(A=float * float(A=@n8, B=1), B=1000)
   n29 = MakeRotator(Pitch=float * float(A="Sin (Radians)"(A=@n27), B=3), Yaw=float * float(A="Sin (Radians)"(A=@n28), B=3))
   n30 = float + float(A=float * float(A=@n9, B=1), B=50)
   n31 = float + float(A=float * float(A=@n9, B=1), B=50)
   n32 = MakeRotator(Pitch=float * float(A="Sin (Radians)"(A=@n30), B=1), Yaw=float * float(A="Sin (Radians)"(A=@n31), B=1))
   n33 = float + float(A=float * float(A=@n9, B=1), B=500)
   n34 = float + float(A=float * float(A=@n9, B=1), B=500)
   n35 = MakeRotator(Pitch=float * float(A="Sin (Radians)"(A=@n33), B=2), Yaw=float * float(A="Sin (Radians)"(A=@n34), B=2))
   n36 = float + float(A=float * float(A=@n9, B=1), B=1000)
   n37 = float + float(A=float * float(A=@n9, B=1), B=1000)
   n38 = MakeRotator(Pitch=float * float(A="Sin (Radians)"(A=@n36), B=3), Yaw=float * float(A="Sin (Radians)"(A=@n37), B=3))
   Set Rot_tentacles(
    execute=[@n102.then, @n6.then]
    Rot_tentacles:Array of Rotators=Make Array!(
     "[0]":Rotator="Scale Rotator (Float)"(A:Rotator=@n23, B:float=@n14) -> (ReturnValue:Rotator=~) {is=title K2Node_PromotableOperator}
     "[1]":Rotator="Scale Rotator (Float)"(A=@n26, B=@n14) -> (ReturnValue=~)
     "[2]":Rotator="Scale Rotator (Float)"(A=@n29, B=@n14) -> (ReturnValue=~)
     "[3]":Rotator="Scale Rotator (Float)"(A=@n32, B=@n14) -> (ReturnValue=~)
     "[4]":Rotator="Scale Rotator (Float)"(A=@n35, B=@n14) -> (ReturnValue=~)
     "[5]":Rotator="Scale Rotator (Float)"(A=@n38, B=@n14) -> (ReturnValue=~)
     "[6]":Rotator="Scale Rotator (Float)"(A=@n17, B=@n14) -> (ReturnValue=~)
     "[7]":Rotator="Scale Rotator (Float)"(A=@n20, B=@n14) -> (ReturnValue=~)
     "[8]":Rotator="Scale Rotator (Float)"(A=@n13, B=@n14) -> (ReturnValue=~)
    ) -> (Array:Array of Rotators)
   ) -> (Output_Get:Array of Rotators)
   n39 = "Sin (Radians)"(A=float + float(A=float * float(A=@n7, B=1.5), B=300))
   n40 = float * float(A="Atan (Radians)"!(A:double=float * float(A=@n39, B=5) {is=title K2Node_CommutativeAssociativeBinaryOperator}) -> (ReturnValue:double), B=5) {is=title K2Node_PromotableOperator}
   n41 = "Sin (Radians)"(A=float + float(A=float * float(A=@n7, B=0.5), B=50))
   n42 = float * float(A="Atan (Radians)"(A=float * float(A=@n41, B=8) {is=title K2Node_CommutativeAssociativeBinaryOperator}), B=5) {is=title K2Node_PromotableOperator}
   n43 = "Sin (Radians)"(A=float + float(A=float * float(A=@n7, B=1), B=50))
   n44 = float * float(A="Atan (Radians)"(A=float * float(A=@n43, B=7) {is=title K2Node_CommutativeAssociativeBinaryOperator}), B=5) {is=title K2Node_PromotableOperator}
   n45 = "Sin (Radians)"(A=float + float(A=float * float(A=@n7, B=1.5), B=600))
   n46 = float * float(A="Atan (Radians)"(A=float * float(A=@n45, B=5) {is=title K2Node_CommutativeAssociativeBinaryOperator}), B=5) {is=title K2Node_PromotableOperator}
   n47 = "Sin (Radians)"(A=float + float(A=float * float(A=@n7, B=2), B=800))
   n48 = float * float(A="Atan (Radians)"(A=float * float(A=@n47, B=10) {is=title K2Node_CommutativeAssociativeBinaryOperator}), B=6) {is=title K2Node_PromotableOperator}
   n49 = "Sin (Radians)"(A=float + float(A=float * float(A=@n7, B=1), B=500))
   n50 = float * float(A="Atan (Radians)"(A=float * float(A=@n49, B=10) {is=title K2Node_CommutativeAssociativeBinaryOperator}), B=6) {is=title K2Node_PromotableOperator}
   n51 = "Sin (Radians)"(A=float + float(A=float * float(A=@n7, B=1), B=600))
   n52 = float * float(A="Atan (Radians)"(A=float * float(A=@n51, B=8) {is=title K2Node_CommutativeAssociativeBinaryOperator}), B=7) {is=title K2Node_PromotableOperator}
   n53 = "Sin (Radians)"(A=float + float(A=float * float(A=@n7, B=1.5), B=50))
   n54 = float * float(A="Atan (Radians)"(A=float * float(A=@n53, B=6) {is=title K2Node_CommutativeAssociativeBinaryOperator}), B=7) {is=title K2Node_PromotableOperator}
   n55 = "Sin (Radians)"(A=float + float(A=float * float(A=@n7, B=2), B=50))
   n56 = float * float(A="Atan (Radians)"(A=float * float(A=@n55, B=5) {is=title K2Node_CommutativeAssociativeBinaryOperator}), B=6) {is=title K2Node_PromotableOperator}
   n57 = "Sin (Radians)"(A=float + float(A=float * float(A=@n7, B=1), B=200))
   n58 = float * float(A="Atan (Radians)"(A=float * float(A=@n57, B=5) {is=title K2Node_CommutativeAssociativeBinaryOperator}), B=6) {is=title K2Node_PromotableOperator}
   n59 = "Sin (Radians)"(A=float + float(A=float * float(A=@n7, B=1), B=50))
   n60 = float * float(A="Atan (Radians)"(A=float * float(A=@n59, B=7) {is=title K2Node_CommutativeAssociativeBinaryOperator}), B=6) {is=title K2Node_PromotableOperator}
   n61 = "Sin (Radians)"(A=float + float(A=float * float(A=@n7, B=1), B=50))
   n62 = float * float(A="Atan (Radians)"(A=float * float(A=@n61, B=7) {is=title K2Node_CommutativeAssociativeBinaryOperator}), B=6) {is=title K2Node_PromotableOperator}
   n63 = "Sin (Radians)"(A=float + float(A=float * float(A=@n7, B=1), B=900))
   n64 = float * float(A="Atan (Radians)"(A=float * float(A=@n63, B=9) {is=title K2Node_CommutativeAssociativeBinaryOperator}), B=5) {is=title K2Node_PromotableOperator}
   n65 = "Sin (Radians)"(A=float + float(A=float * float(A=@n7, B=1.5), B=200))
   n66 = float * float(A="Atan (Radians)"(A=float * float(A=@n65, B=5) {is=title K2Node_CommutativeAssociativeBinaryOperator}), B=5) {is=title K2Node_PromotableOperator}
   n67 = "Sin (Radians)"(A=float + float(A=float * float(A=@n7, B=0.5), B=50))
   n68 = float * float(A="Atan (Radians)"(A=float * float(A=@n67, B=8) {is=title K2Node_CommutativeAssociativeBinaryOperator}), B=5) {is=title K2Node_PromotableOperator}
   n69 = "Sin (Radians)"(A=float + float(A=float * float(A=@n7, B=1), B=60))
   n70 = float * float(A="Atan (Radians)"(A=float * float(A=@n69, B=8) {is=title K2Node_CommutativeAssociativeBinaryOperator}), B=5) {is=title K2Node_PromotableOperator}
   n71 = "Sin (Radians)"(A=float + float(A=float * float(A=@n7, B=1), B=500))
   n72 = float * float(A="Atan (Radians)"(A=float * float(A=@n71, B=5) {is=title K2Node_CommutativeAssociativeBinaryOperator}), B=6) {is=title K2Node_PromotableOperator}
   n73 = "Sin (Radians)"(A=float + float(A=float * float(A=@n7, B=1.5), B=100))
   n74 = float * float(A="Atan (Radians)"(A=float * float(A=@n73, B=5) {is=title K2Node_CommutativeAssociativeBinaryOperator}), B=6) {is=title K2Node_PromotableOperator}
   > Set Rot_eyes(Rot_eyes:Array of Rotators=Make Array(
     "[0]"=MakeRotator(Pitch=@n68, Yaw=@n70)
     "[1]"=MakeRotator(Pitch=@n72, Yaw=@n74)
     "[2]"=MakeRotator(Pitch=@n40, Yaw=@n42)
     "[3]"=MakeRotator(Pitch=@n66, Yaw=@n64)
     "[4]"=MakeRotator(Pitch=@n46, Yaw=@n44)
     "[5]"=MakeRotator(Pitch=@n54, Yaw=@n52)
     "[6]"=MakeRotator(Pitch=@n58, Yaw=@n56)
     "[7]"=MakeRotator(Pitch=@n62, Yaw=@n60)
     "[8]"=MakeRotator(Pitch=@n48, Yaw=@n50)
    )) -> (Output_Get:Array of Rotators)
   n75 = float * float(A=@n7, B=0.5)
   n76 = "Atan (Radians)"(A=float * float(A="Sin (Radians)"(A=@n75), B=60))
   n77 = float * float(A=float * float(A=@n76, B=@n76), B=-0.415)
   n78 = float * float(A=@n7, B=1)
   n79 = "Atan (Radians)"(A=float * float(A="Sin (Radians)"(A=@n78), B=10))
   n80 = float * float(A=float * float(A=@n79, B=@n79), B=-0.46)
   n81 = float * float(A=@n7, B=0.75)
   n82 = "Atan (Radians)"(A=float * float(A="Sin (Radians)"(A=@n81), B=13.33))
   n83 = float * float(A=float * float(A=@n82, B=@n82), B=-0.445)
   n84 = float * float(A=@n7, B=0.45)
   n85 = "Atan (Radians)"(A=float * float(A="Sin (Radians)"(A=@n84), B=66.66))
   n86 = float * float(A=float * float(A=@n85, B=@n85), B=-0.413)
   n87 = float * float(A=@n7, B=0.7)
   n88 = "Atan (Radians)"(A=float * float(A="Sin (Radians)"(A=@n87), B=14.28))
   n89 = float * float(A=float * float(A=@n88, B=@n88), B=-0.44)
   n90 = float * float(A=@n7, B=0.8)
   n91 = "Atan (Radians)"(A=float * float(A="Sin (Radians)"(A=@n90), B=12.5))
   n92 = float * float(A=float * float(A=@n91, B=@n91), B=-0.45)
   n93 = float * float(A=@n7, B=0.9)
   n94 = "Atan (Radians)"(A=float * float(A="Sin (Radians)"(A=@n93), B=11.11))
   n95 = float * float(A=float * float(A=@n94, B=@n94), B=-0.455)
   n96 = float * float(A=@n7, B=0.4)
   n97 = "Atan (Radians)"(A=float * float(A="Sin (Radians)"(A=@n96), B=75))
   n98 = float * float(A=float * float(A=@n97, B=@n97), B=-0.415)
   n99 = float * float(A=@n7, B=0.55)
   n100 = "Atan (Radians)"(A=float * float(A="Sin (Radians)"(A=@n99), B=54.54))
   n101 = float * float(A=float * float(A=@n100, B=@n100), B=-0.415)
   > Set Rot_lids(Rot_lids:"Array of Float (double-precision)s"=Make Array(
     "[0]":double=float + float(A=@n80, B=1)
     "[1]":double=float + float(A=@n89, B=1)
     "[2]":double=float + float(A=@n92, B=1)
     "[3]":double=float + float(A=@n95, B=1)
     "[4]":double=float + float(A=@n83, B=1)
     "[5]":double=float + float(A=@n98, B=1)
     "[6]":double=float + float(A=@n101, B=1)
     "[7]":double=float + float(A=@n86, B=1)
     "[8]":double=float + float(A=@n77, B=1)
    ) -> (Array:"Array of Float (double-precision)s")) -> (Output_Get:"Array of Float (double-precision)s")
   Set Amp_sin(execute=@n5.else, Amp_sin=5)
   > n102 = Set Freq_sin(Freq_sin=3)
   TryGetPawnOwner()
   n103 = Get Speed_Fin() -> (Speed_Fin:double)
   n104 = Get Freq_sin()
   FInterpTo(Current=@n104, Target=1, DeltaTime=@n4.DeltaTimeX, InterpSpeed=@n103)
   n105 = Get Speed_Str() -> (Speed_Str:double)
   n106 = Get Amp_noise()
   FInterpTo(Current=@n106, Target=50, DeltaTime=@n4.DeltaTimeX, InterpSpeed=@n105)
   FInterpTo(Current=@n106, Target=30, DeltaTime=@n4.DeltaTimeX, InterpSpeed=@n103)
   FInterpTo(Current=@n104, Target=1.5, DeltaTime=@n4.DeltaTimeX, InterpSpeed=@n105)
   n107 = Get Amp_sin()
   FInterpTo(Current=@n107, Target=20, DeltaTime=@n4.DeltaTimeX, InterpSpeed=@n105)
   FInterpTo(Current=@n107, Target=5, DeltaTime=@n4.DeltaTimeX, InterpSpeed=@n103)
   n108 = Get Freq_noise()
   FInterpTo(Current=@n108, Target=5, DeltaTime=@n4.DeltaTimeX, InterpSpeed=@n105)
   FInterpTo(Current=@n108, Target=7, DeltaTime=@n4.DeltaTimeX, InterpSpeed=@n103)
  ]}
]
variables [
 Amp_noise {type=number}
 Amp_sin {type=number}
 Freq_noise {type=number}
 Freq_sin {type=number}
 Gravity overraide {type=Vector}
 Rot_eyes {type="array<Rotator>", default []}
 Rot_lids {type="array<number>", default []}
 Rot_tentacle_L_up {type=Rotator}
 Rot_tentacle_R_bot {type=Rotator}
 Rot_tentacle_R_up {type=Rotator}
 Rot_tentacles {type="array<Rotator>", default []}
 Speed_Fin {type=number, default=10}
 Speed_Str {type=number, default=0.5}
]
